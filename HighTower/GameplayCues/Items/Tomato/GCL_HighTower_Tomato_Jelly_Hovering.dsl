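; Gameplay cue for Tomato Jelly jetpack hovering, a child of the jetpack hovering cue that drives its effects and audio.
; OnBurst attaches backpack jet particles, sets material parameters and enables audio, with first-jump sounds re-armed on landing.
; A downward trace from the backpack positions ground-interaction particles and scales them by hit distance.
; Cease, removal and destruction handle cleanup, including a fix for a stuck out-of-fuel loop.
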(event Custom|On%20Pawn%20Landed
  (params (Hit :type Hit%20Result%20Structure%20%28by%20ref%29))
  ;[1 1 1] Reset "First-Jump" flag once we've landed
  (Variables|Sounds|SetbIsFirstJump true)
  (EventDispatchers|UnbindEventfromOnPawnLanded (Variables|GameplayCueNotify_Jetpack_Hovering|GetPlayerPawn) (delegate Custom|On%20Pawn%20Landed)))

(event Custom|ResetFX
  (bind _groundinteract (Variables|Default|GetGroundInteract))
  (bind _groundinteract_2 (Variables|Default|GetGroundInteract))
  ;[1 1 1] FX Control
  (Utilities|FlowControl|Sequence
    (:then_0
      (Utilities|FlowControl|DoOnce false
        (:Completed
          (Utilities|IsValid _groundinteract_2
            (:Is%20Valid
              (Components|Activation|Deactivate _groundinteract_2))
            (:Is%20Not%20Valid
              _))
          (bind _returnvalue (Effects|SpawnEmitterAttached :EmitterTemplate "/FNGameplayCues/Athena/Items/Gameplay/BackPacks/BoostJumpPack/FX/P_Jetpack_GroundInteraction.P_Jetpack_GroundInteraction" :AttachToComponent (Customization|GetSkeletalMeshforPartType (Variables|GameplayCueNotify_Jetpack_Hovering|GetPlayerPawn) "Head" :!class /Script/FortniteGame.FortPlayerPawn) :AttachPointName "GroundInteract" :Scale "1.000000,1.000000,1.000000" :LocationType "SnapToTarget" :bAutoDestroy true :PoolingMethod "None" :bAutoActivate true))
          (Variables|Default|SetGroundInteract _returnvalue))))
    (:then_1
      (Utilities|IsValid _groundinteract
        (:Is%20Valid
          (Components|Activation|Deactivate _groundinteract))))))

(event GameplayCueNotify|EventHandleGameplayCue
  (params (MyTarget :type Actor%20Object%20Reference) (EventType :type EGameplayCueEvent%20Enum :default "OnActive") (Parameters :type Gameplay%20Cue%20Parameters%20Structure%20%28by%20ref%29))
  (bind _outoffuelaudiocomp (Variables|GameplayCueNotify_Jetpack_Hovering|GetOutOfFuelAudioComp))
  ;[1 1 1] Fix for loop getting stuck
  (Utilities|IsValid MyTarget
    (:Is%20Valid
      (if (Utilities|Enum|Equal(Enum) EventType "Removed")
        (Utilities|IsValid _outoffuelaudiocomp
          (:Is%20Valid
            (if (Audio|Components|Audio|IsPlaying _outoffuelaudiocomp)
              (Audio|Components|Audio|Stop _outoffuelaudiocomp))))))))

(event EventTryToPlayGroundEffects
  (bind _groundinteract (Variables|Default|GetGroundInteract))
  (bind _groundinteract_2 (Variables|Default|GetGroundInteract))
  ;[1 1 1] FX Control
  (bind (_outhit _returnvalue) (Collision|LineTraceForObjects :Start (Transformation|GetSocketLocation (Customization|GetSkeletalMeshforPartType (Variables|GameplayCueNotify_Jetpack_Hovering|GetPlayerPawn) "Backpack" :!class /Script/FortniteGame.FortPlayerPawn) "GroundInteract") :End (+ (Transformation|GetSocketLocation (Customization|GetSkeletalMeshforPartType (Variables|GameplayCueNotify_Jetpack_Hovering|GetPlayerPawn) "Backpack" :!class /Script/FortniteGame.FortPlayerPawn) "GroundInteract") "0, 0,-650.000000") :ObjectTypes (Utilities|Array|MakeArray "ObjectTypeQuery1") :bTraceComplex false :DrawDebugType "None" :bIgnoreSelf true :TraceColor "(R=1.000000,G=0.000000,B=0.000000,A=1.000000)" :TraceHitColor "(R=0.000000,G=1.000000,B=0.000000,A=1.000000)" :DrawTime 1.0))
  (if _returnvalue
    (Utilities|FlowControl|Sequence
      (:then_0
        (Utilities|FlowControl|DoOnce false
          (:Completed
            (Utilities|IsValid _groundinteract_2
              (:Is%20Valid
                (Components|Activation|Deactivate _groundinteract_2))
              (:Is%20Not%20Valid
                _))
            (bind _returnvalue_2 (Effects|SpawnEmitterAttached :EmitterTemplate "/FNGameplayCues/Athena/Items/Gameplay/BackPacks/BoostJumpPack/FX/P_Jetpack_GroundInteraction.P_Jetpack_GroundInteraction" :AttachToComponent (Customization|GetSkeletalMeshforPartType (Variables|GameplayCueNotify_Jetpack_Hovering|GetPlayerPawn) "Head" :!class /Script/FortniteGame.FortPlayerPawn) :AttachPointName "GroundInteract" :Scale "1.000000,1.000000,1.000000" :LocationType "SnapToTarget" :bAutoDestroy true :PoolingMethod "None" :bAutoActivate true))
            (Variables|Default|SetGroundInteract _returnvalue_2))))
      (:then_1
        (Utilities|IsValid (Variables|Default|GetGroundInteract)
          (:Is%20Valid
            (Transformation|SetWorldLocationAndRotation (Variables|Default|GetGroundInteract) (.Location (Collision|BreakHitResult _outhit)) (Math|Conversions|RotationFromXVector (.ImpactNormal (Collision|BreakHitResult _outhit))) false false)
            (Effects|Components|ParticleSystem|SetFloatParameter (Variables|Default|GetGroundInteract) "Height" (/ (Math|Vector|VectorLength (- (.Location (Collision|BreakHitResult _outhit)) (.TraceStart (Collision|BreakHitResult _outhit)))) 6.5e+02)))))
      (:then_2
        (Utilities|IsValid (Variables|Default|GetGroundInteract)
          (:Is%20Valid
            (Components|Activation|Activate (Variables|Default|GetGroundInteract) false)))))
    (else
      (Utilities|IsValid _groundinteract
        (:Is%20Valid
          (Components|Activation|Deactivate _groundinteract))))))

(event EventDestroyed
  (bind _outoffuelaudiocomp (Variables|GameplayCueNotify_Jetpack_Hovering|GetOutOfFuelAudioComp))
  ;[1 1 1] Fix for loop getting stuck
  (Utilities|IsValid _outoffuelaudiocomp
    (:Is%20Valid
      (Audio|Components|Audio|Stop _outoffuelaudiocomp))))

(event EventPlayLowFuelWarning
  ;[1 1 1] Low Fuel Warning
  (Audio|PlaySound2D (Variables|GameplayCueNotify_Jetpack_Hovering|GetPlayerPawn) (Variables|Sounds|GetSoundOnFuelWarning)))

(fn OnBurst
  (params (MyTarget :type Actor%20Object%20Reference) (Parameters :type Gameplay%20Cue%20Parameters%20Structure%20%28by%20ref%29))
  (bind _playerpawn (Variables|GameplayCueNotify_Jetpack_Hovering|GetPlayerPawn))
  (bind _returnvalue (Customization|GetSkeletalMeshforPartType _playerpawn "Backpack" :!class /Script/FortniteGame.FortPlayerPawn))
  (bind _returnvalue_2 (|CallParentFunction MyTarget Parameters))
  (if _returnvalue_2
    (Utilities|IsValid _returnvalue
      (:Is%20Valid
        (bind _returnvalue_3 (Effects|SpawnEmitterAttached 0 _returnvalue "Jet1"))
        (bind _output_get (Variables|GameplayCueNotify_Jetpack_Hovering|SetParticleSystemComp _returnvalue_3))
        (Effects|Components|ParticleSystem|SetVectorParameter _output_get "LightSize" (select (Utilities|Equal(SoftObjectReference) (Utilities|ResolveSoftReference (Variables|/Script/Engine.SkinnedMeshComponent|GetSkeletalMesh _returnvalue)) (Variables|Default|GetMilo%20Backpack)) (Variables|Default|GetMilo_LightScale) "1.000000,1.000000,1.000000"))
        (HelperMethods|SetJetpackAudioEnabled true)
        (Rendering|Material|SetScalarParameterValueonMaterials _returnvalue "FullBlink" 1.0)
        (Rendering|Material|SetScalarParameterValueonMaterials _returnvalue "Refueling" 0.0)
        (return true))
      (:Is%20Not%20Valid
        (return false)))
    (else
      (return false))))

(fn OnCeaseRelevant
  (params (MyTarget :type Actor%20Object%20Reference) (Parameters :type Gameplay%20Cue%20Parameters%20Structure%20%28by%20ref%29))
  (EventDispatchers|UnbindEventfromOnPawnLanded (Variables|GameplayCueNotify_Jetpack_Hovering|GetPlayerPawn) (EventDispatchers|CreateEvent :!event On%20Pawn%20Landed))
  (Utilities|FlowControl|Sequence
    (:then_0
      (CallFunction|ResetFX)
      (Utilities|IsValid (Variables|Default|GetGroundInteract)
        (:Is%20Valid
          (Components|Activation|Deactivate (Variables|Default|GetGroundInteract)))))
    (:then_1
      (HelperMethods|SetJetpackAudioEnabled false)
      (|CallParentFunction (Variables|GetMyTarget) (Variables|GetParameters))
      (return true))))

(fn SetJetpackAudioEnabled
  (params (Enabled :type Boolean))
  (bind _outoffuelaudiocomp (Variables|GameplayCueNotify_Jetpack_Hovering|GetOutOfFuelAudioComp))
  (bind _playerpawn (Variables|GameplayCueNotify_Jetpack_Hovering|GetPlayerPawn))
  (bind _outoffuelaudiocomp_2 (Variables|GameplayCueNotify_Jetpack_Hovering|GetOutOfFuelAudioComp))
  (Variables|Sounds|SetbJetpackAudioEnabled Enabled)
  (Actor|Tick|SetActorTickEnabled (Variables|Sounds|GetbJetpackAudioEnabled))
  (Utilities|IsValid (Variables|GameplayCueNotify_Jetpack_Hovering|GetPlayerPawn)
    (:Is%20Valid
      (if (Variables|Sounds|GetbJetpackAudioEnabled)
        (EventDispatchers|BindEventtoOnPawnLanded (Variables|GameplayCueNotify_Jetpack_Hovering|GetPlayerPawn) (EventDispatchers|CreateEvent :!event On%20Pawn%20Landed))
        (Audio|SpawnSoundAttached (select (Variables|Sounds|GetbIsFirstJump) (Variables|Sounds|GetSoundOnStart) (Variables|SoundIndicator|GetSoundOnThrust)) (Variables|/Script/Engine.Actor|GetRootComponent (Variables|GameplayCueNotify_Jetpack_Hovering|GetPlayerPawn)))
        (Variables|Sounds|SetbIsFirstJump false)
        (Utilities|FlowControl|Sequence
          (:then_0
            (bind _returnvalue (Audio|SpawnSoundAttached :Sound (Variables|Sounds|GetSoundLoopOnIdle) :AttachToComponent (Variables|/Script/Engine.Actor|GetRootComponent (Variables|GameplayCueNotify_Jetpack_Hovering|GetPlayerPawn)) :AttachPointName "None" :LocationType "KeepRelativeOffset" :bStopWhenAttachedToDestroyed false :VolumeMultiplier 1.0 :PitchMultiplier 1.0 :StartTime (Math|Random|RandomFloatinRange 0.0 7.4) :bAutoDestroy true))
            (Variables|GameplayCueNotify_Jetpack_Hovering|SetIdleAudioComp _returnvalue))
          (:then_1
            (bind _returnvalue_2 (Audio|SpawnSoundAttached :Sound (Variables|Sounds|GetSoundLoopOnAccel) :AttachToComponent (Variables|/Script/Engine.Actor|GetRootComponent (Variables|GameplayCueNotify_Jetpack_Hovering|GetPlayerPawn)) :AttachPointName "None" :LocationType "KeepRelativeOffset" :bStopWhenAttachedToDestroyed false :VolumeMultiplier 1.0 :PitchMultiplier 1.0 :StartTime (Math|Random|RandomFloatinRange 0.0 5.0) :bAutoDestroy true))
            (Variables|GameplayCueNotify_Jetpack_Hovering|SetAccelAudioComp _returnvalue_2))
          (:then_2
            (if (Pawn|IsLocallyControlled _playerpawn)
              (bind _returnvalue_3 (Audio|SpawnSoundAttached :Sound (Variables|Sounds|GetSoundLoopOnOutOfFuel) :AttachToComponent (Variables|/Script/Engine.Actor|GetRootComponent _playerpawn) :AttachPointName "None" :LocationType "KeepRelativeOffset" :bStopWhenAttachedToDestroyed false :VolumeMultiplier 0.0 :PitchMultiplier 1.0 :StartTime (Math|Random|RandomFloatinRange 0.0 5.5) :bAutoDestroy true))
              (Variables|GameplayCueNotify_Jetpack_Hovering|SetOutOfFuelAudioComp _returnvalue_3))))
        (else
          _)))
    (:Is%20Not%20Valid
      _))
  (Audio|Components|Audio|FadeOut (Variables|GameplayCueNotify_Jetpack_Hovering|GetIdleAudioComp) 0.25 0.0 "Linear" :!class /Script/Engine.AudioComponent)
  (Utilities|IsValid _outoffuelaudiocomp
    (:Is%20Valid
      (Audio|Components|Audio|Stop _outoffuelaudiocomp))
    (:Is%20Not%20Valid
      _))
  (Utilities|IsValid (Variables|GameplayCueNotify_Jetpack_Hovering|GetPlayerPawn)
    (:Is%20Valid
      (Utilities|IsValid _outoffuelaudiocomp_2
        (:Is%20Valid
          (Audio|SpawnSoundAttached (select (and (>= (Variables|/Script/Engine.AudioComponent|GetVolumeMultiplier _outoffuelaudiocomp_2) 0.95) (> (- (Utilities|Time|GetGameTimeinSeconds (Variables|GameplayCueNotify_Jetpack_Hovering|GetPlayerPawn)) (Variables|GameplayCueNotify_Jetpack_Hovering|GetActivationTime)) 4.0)) (Variables|Sounds|GetSoundOnOutOfFuel) (Variables|Sounds|GetSoundOnStop)) (Variables|/Script/Engine.Actor|GetRootComponent (Variables|GameplayCueNotify_Jetpack_Hovering|GetPlayerPawn))))))))

(fn ConstructionScript)
