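; Burst gameplay cue for the HighTower Tomato Repulsor Cannon blast, aiming its Cascade beam effect along the player's aim.
; On burst, it gets the target's player controller and queries an aim point at a fixed range, ignoring the target itself.
; It sets the first particle component's BeamEnd to that point and rotates it from the target's location toward it.
; A comment marks this as temporary because the gameplay cue parameters return null.
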(fn OnBurst(Cascade)
  (params (MyTarget :type Actor%20Object%20Reference) (Parameters :type Gameplay%20Cue%20Parameters%20Structure%20%28by%20ref%29) (ParticleComponents :type Array%20of%20Cascade%20Particle%20System%20Component%20Object%20References) (AudioComponents :type Array%20of%20Audio%20Component%20Object%20References) (BurstCameraShakeInstance :type Camera%20Shake%20Base%20Object%20Reference) (BurstDecalInstance :type Decal%20Actor%20Object%20Reference))
  (bind _output (Utilities|Array|Get(acopy) ParticleComponents))
  ;[1 0.08056 0.04076] This is TEMP. for some reason the gameplay params are returning null
  (bind _returnvalue (Fortnite|GetPlayerAimPoint (Fortnite|GetFortPlayerControllerfromActor MyTarget) 1e+04 (Utilities|Array|MakeArray MyTarget)))
  (Effects|Components|ParticleSystem|SetVectorParameter _output "BeamEnd" _returnvalue)
  (Transformation|SetWorldRotation _output (Math|Rotator|FindLookatRotation (Transformation|GetActorLocation MyTarget) _returnvalue) false false)
  (return))
